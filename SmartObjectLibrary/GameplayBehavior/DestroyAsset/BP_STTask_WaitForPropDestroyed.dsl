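; State Tree task that waits until a target building prop is destroyed, used in a Smart Object destroy-asset behavior.
; On entering the state it casts its Actor input to a BuildingActor and caches it, failing the task if the cast fails.
; Each tick it checks whether that actor's health has reached zero or it is no longer valid, then calls DestroyActor on it and finishes successfully.
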(event EventEnterState
  (params (Transition :type State%20Tree%20Transition%20Result%20Structure%20%28by%20ref%29))
  (bind _asbuilding_actor (Utilities|Casting|CastToBuildingActor (Variables|Default|GetActor))
    (:then
      (Variables|Default|SetBuildingActor _asbuilding_actor))
    (:CastFailed
      (StateTree|FinishTask false))))

(event EventTick
  (params (DeltaTime :type Float%20%28single-precision%29))
  (if (or (<= (Health|GetHealth (Variables|Default|GetBuildingActor) :!class /Script/FortniteGame.BuildingActor) 0) (not (Utilities|IsValid (Variables|Default|GetBuildingActor) :!class /Script/Engine.KismetSystemLibrary)))
    ; Force Disable Components before exiting behavior. Weird Logic were Component Stays, could be related to the prop I used.
    (Actor|DestroyActor (Variables|Default|GetBuildingActor))
    (StateTree|FinishTask true)))
