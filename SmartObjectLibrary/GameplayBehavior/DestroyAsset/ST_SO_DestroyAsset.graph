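# Smart object StateTree that has a bot NPC destroy a prop with its pickaxe.
# A global task gets the bot controller, and Root sets weapon holster support before EquipPickAxe fetches the inventory item, equips it and plays a speech bubble.
# UsePickAxe then runs the fire-weapon task in HoldFire mode while its child state waits for the prop's destruction, which completes the tree as succeeded.
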StateTree /SmartObjectLibrary/GameplayBehavior/DestroyAsset/ST_SO_DestroyAsset
global_tasks [{
  id=#1
  description=Get BotController
  properties {
   iD=#2
   instance {_structType=/Script/FortniteAIServer.FortStateTreeGetBotControllerTaskInstanceData, botController=null, pawn=null}
   node {
    name=None
    _structType=/Script/FortniteAIServer.FortStateTreeGetBotControllerTask
    bCanEditConsideredForCompletion=true
    bConsideredForCompletion=true
    bHasTaskCompletionDelegateDispatcher=false
    bTaskEnabled=true
    transitionHandlingPriority=Normal
    bindingsBatch {value=65535}
    executionRuntimeTemplateIndex {value=65535}
    instanceDataHandle {index=65535, source=None, stateHandle {index=65535}}
    instanceTemplateIndex {value=65535}
    outputBindingsBatch {value=65535}
   }
  }
 }]
states [
 Root {
  id=#3
  type=STATE
  children [#4, #5]
  tasks [{
    id=#6
    description=Set Support Holstered Weapon
    properties {
     iD=#7
     instance {_structType=/Script/FortniteAIServer.FortStateTreeBotSetSupportHolsteredWeaponInstanceData, bSupportHolstering=false, controller=null}
     node {
      name=None
      _structType=/Script/FortniteAIServer.FortStateTreeBotSetSupportHolsteredWeapon
      bCanEditConsideredForCompletion=true
      bConsideredForCompletion=true
      bHasTaskCompletionDelegateDispatcher=false
      bTaskEnabled=true
      transitionHandlingPriority=Normal
      bindingsBatch {value=65535}
      executionRuntimeTemplateIndex {value=65535}
      instanceDataHandle {index=65535, source=None, stateHandle {index=65535}}
      instanceTemplateIndex {value=65535}
      outputBindingsBatch {value=65535}
     }
    }
   }]
  transitions []
 }
 EquipPickAxe {
  id=#4
  type=STATE
  tasks [
   {
    id=#8
    description=Get Item From Inventory
    properties {
     iD=#9
     instance {
      _structType=/Script/FortniteAIServer.FortStateTreeGetItemFromInventoryTaskInstanceData
      controller=null
      item=null
      itemVariant {item=null, itemVariantGuid=#10}
     }
     node {
      name=None
      _structType=/Script/FortniteAIServer.FortStateTreeGetItemFromInventoryTask
      bCanEditConsideredForCompletion=true
      bConsideredForCompletion=true
      bHasTaskCompletionDelegateDispatcher=false
      bTaskEnabled=true
      transitionHandlingPriority=Normal
      bindingsBatch {value=65535}
      executionRuntimeTemplateIndex {value=65535}
      instanceDataHandle {index=65535, source=None, stateHandle {index=65535}}
      instanceTemplateIndex {value=65535}
      outputBindingsBatch {value=65535}
     }
    }
   }
   {
    id=#11
    description=Equip Item
    properties {
     iD=#12
     instance {_structType=/Script/FortniteAIServer.FortStateTreeBotEquipItemTaskInstanceData, bValidateAbility=false, controller=null, itemToEquip=null}
     node {
      name=None
      _structType=/Script/FortniteAIServer.FortStateTreeBotEquipItemTask
      bCanEditConsideredForCompletion=true
      bConsideredForCompletion=true
      bHasTaskCompletionDelegateDispatcher=false
      bTaskEnabled=true
      transitionHandlingPriority=Normal
      bindingsBatch {value=65535}
      executionRuntimeTemplateIndex {value=65535}
      instanceDataHandle {index=65535, source=None, stateHandle {index=65535}}
      instanceTemplateIndex {value=65535}
      outputBindingsBatch {value=65535}
     }
    }
   }
   {
    id=#13
    description=BP STTask Play Speech Bubble
    properties {
     iD=#14
     instanceObject=&self:StateTreeEditorData_0.StateTreeState_1.BP_STTask_PlaySpeechBubble_C_0
     node {
      name=None
      _structType=&StateTreeBlueprintTaskWrapper
      taskClass=/NPCCore/StateTreeTasks/BP_STTask_PlaySpeechBubble.BP_STTask_PlaySpeechBubble_C
     }
    }
   }
  ]
  transitions [{id=#15, trigger=ON_STATE_COMPLETED, state {id=#5, name=UsePickAxe}}]
 }
 UsePickAxe {
  id=#5
  type=STATE
  children [#16]
  tasks [
   {
    id=#17
    description=Bot Get Current Weapon
    properties {
     iD=#18
     instance {_structType=/Script/FortniteAIServer.FortStateTreeBotGetCurrentWeaponTaskInstanceData, botController=null, currentWeapon=null}
     node {
      name=None
      _structType=/Script/FortniteAIServer.FortStateTreeBotGetCurrentWeaponTask
      bCanEditConsideredForCompletion=true
      bConsideredForCompletion=true
      bHasTaskCompletionDelegateDispatcher=false
      bTaskEnabled=true
      transitionHandlingPriority=Normal
      bindingsBatch {value=65535}
      executionRuntimeTemplateIndex {value=65535}
      instanceDataHandle {index=65535, source=None, stateHandle {index=65535}}
      instanceTemplateIndex {value=65535}
      outputBindingsBatch {value=65535}
     }
    }
   }
   {
    id=#19
    description=Bot Fire Weapon
    properties {
     iD=#20
     instance {_structType=/Script/FortniteAIServer.FortStateTreeBotFireWeaponTaskInstanceData, behavior=HoldFire, botController=null, firingTimes []}
     node {
      name=None
      _structType=/Script/FortniteAIServer.FortStateTreeBotFireWeaponTask
      bCanEditConsideredForCompletion=true
      bConsideredForCompletion=true
      bHasTaskCompletionDelegateDispatcher=false
      bTaskEnabled=true
      transitionHandlingPriority=Normal
      bindingsBatch {value=65535}
      executionRuntimeTemplateIndex {value=65535}
      instanceDataHandle {index=65535, source=None, stateHandle {index=65535}}
      instanceTemplateIndex {value=65535}
      outputBindingsBatch {value=65535}
     }
    }
   }
  ]
  transitions []
 }
 WaitUntilAssetDestroyed {
  id=#16
  type=STATE
  tasks [{
    id=#21
    description=BP STTask Wait for Prop Destroyed
    properties {
     iD=#22
     instanceObject=&self:StateTreeEditorData_0.StateTreeState_4.BP_STTask_WaitForPropDestroyed_C_0
     node {
      name=None
      _structType=&StateTreeBlueprintTaskWrapper
      taskClass=&/BP_STTask_WaitForPropDestroyed.BP_STTask_WaitForPropDestroyed_C
     }
    }
   }]
  transitions [{id=#23, trigger=ON_STATE_COMPLETED, state {id=#24, name=Root, linkType=SUCCEEDED}}]
 }
]
